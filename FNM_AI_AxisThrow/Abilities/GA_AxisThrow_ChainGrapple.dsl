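; AI bot ability that throws a homing chain hook at its blackboard target when within a set distance range.
; It equips chains and plays a montage whose notify runs target selection and spawns a hookshot projectile with a hit effect container.
; A touch jumps to PullSuccess and ejects a vehicle passenger or ends swinging/skydiving; a timeout jumps to PullFail.
; EndAbility commits the ability, re-equips the best weapon and disables the projectile.
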(event Ability|EventActivateAbility
  (bind (_montageinfo_montagetoplay _montageinfo_animplayrate _montageinfo_animrootmotiontranslationscale) (Utilities|Struct|BreakFortGameplayAbilityMontageInfo (Variables|GetMontageInfo_MontageToPlay)))
  (bind _outfortplayerpawn (CallFunction|GetAvatarasValidFortPlayerPawn)
    (:ValidFortPlayerPawn
      (bind _output_get (Variables|InternalData|SetAIPawn _outfortplayerpawn))
      (bind _asfort_athena_aibot_controller (Utilities|Casting|CastToFortAthenaAIBotController (Pawn|GetController _output_get))
        (:then
          (bind _output_get_2 (Variables|InternalData|SetAIController _asfort_athena_aibot_controller))
          (FortWeapon|ModAmmo (Variables|/Script/FortniteGame.FortPawn|GetCurrentWeapon (Variables|InternalData|GetAIPawn)) 500)
          (Class|FortAthenaAIBotController|EquipWeaponbyTag _output_get_2 (Variables|Default|GetWeaponTag_Chains))
          (Utilities|FlowControl|Delay 0.5)
          (bind (_asynctaskproxy _notifytag) (Ability|Tasks|PlayMontageAndWaitNotify "None" _montageinfo_montagetoplay _montageinfo_animplayrate "None" true _montageinfo_animrootmotiontranslationscale)
            (:then
              (Class|BPFLAICommonFunctions|PlaySpeechBubble (Variables|InternalData|GetAIPawn) (Variables|Default|GetSpeechBubbleTag) true true)
              (Utilities|FlowControl|Delay (ScalableFloat|GetValueatLevel (Variables|Default|GetWhipTime)))
              (Ability|Animation|MontageJumptoSection "Throw")
              (Variables|Default|SetbIsPulling false))
            (:OnCompleted
              _)
            (:OnInterrupted
              _)
            (:OnCancelled
              _)
            (:OnNotifyBegin
              (if (== _notifytag (Variables|Default|GetTag_Anim_ThrowHook))
                ;[1 1 1] Try Hit Target and Spawn Projectile
                (bind (_asynctaskproxy_2 _targetdata) (Ability|Tasks|DoTargetSelection "None" "(TagName=\"Targeting.ApplicationTag.1\")")
                  (:then
                    (Utilities|FlowControl|Delay (ScalableFloat|GetValueatLevel (Variables|Default|GetWaitTimeToPullWhenMissing)))
                    (if (not (Variables|Default|GetbIsPulling))
                      (Default|DisableProjectile)
                      (Ability|Animation|MontageJumptoSection "PullFail")))
                  (:Targeted
                    _)
                  (:Cancelled
                    _))
                (bind (_asynctaskproxy_3 _projectiledata) (Projectile|Tasks|SpawnProjectileAndWait :Class (Variables|Default|GetHookProjectileClass) :RequestedBy (CallFunction|GetAvatarasFortPlayerPawn) :SpawnLocation (CallFunction|GetProjectileSpawnLocation) :SpawnRotation (Transformation|GetActorRotation (CallFunction|GetAvatarasFortPlayerPawn)) :SpawnDirection (Targeting|GetDefaultTargetingRotation (CallFunction|GetAvatarasFortPlayerPawn)) :InitialSpeed 8e+03 :EffectContainerSpecToApplyOnHit (Ability|MakeGameplayEffectContainerSpec "(TagName=\"Gameplay.Application.Hit\")") :GravityScale 0.0 :bAllowProjectileSpawnWhenDead false :bAllowProjectileSpawnWhenDBNO false)
                  (:then
                    ;;[1 1 1] On the server, bind to know when the projectile is killed. This fires when the projectile has traveled beyond its kill distance.
                    (if (.bSuccess (Utilities|Casting|CastToFortPawn (.HitActor (Collision|BreakHitResult (Ability|TargetData|GetHitResultfromTargetData _targetdata)))))
                      (bind (_projectiledata_spawnedprojectile _projectiledata_hits) (Utilities|Struct|BreakProjectileEventData _projectiledata))
                      (Projectile|SetHomingData _projectiledata_spawnedprojectile _projectiledata_spawnedprojectile (Utilities|Struct|MakeProjectileHomingData "LockOn" (ScalableFloat|GetValueatLevel (Variables|Default|GetHomingMinTurnSpeed)) (ScalableFloat|GetValueatLevel (Variables|Default|GetHomingMaxTurnSpeed)) 0.15 0.0 -1 (.HitActor (Collision|BreakHitResult (Ability|TargetData|GetHitResultfromTargetData _targetdata)))) :!class /Script/FortniteGame.FortProjectileMovementInterface)))
                  (:Created
                    (bind _asb_prj_hookshot (Utilities|Casting|CastToB_Prj_Hookshot _projectiledata_spawnedprojectile))
                    (bind _output_get_3 (Variables|Default|SetHookshotProjectile _asb_prj_hookshot))
                    (Default|BindEventtoOnProjectileKilled _output_get_3 (EventDispatchers|CreateEvent :!event Custom|HandeProjectileKilled) :!class /ThornRipper/Gameplay/Items/HookshotGameplay/Gameplay/B_Prj_Hookshot.B_Prj_Hookshot_C))
                  (:Touched
                    (Ability|Animation|MontageJumptoSection "PullSuccess")
                    (Variables|Default|SetbIsPulling true)
                    (Default|DisableProjectile)
                    ;[1 1 1] If player hit, reset Life Drain Cooldown
                    (if (> (Utilities|Array|Length _projectiledata_hits) 0)
                      (bind _asfort_athena_vehicle (Utilities|Casting|CastToFortAthenaVehicle (.HitActor (Collision|BreakHitResult (Utilities|Array|Get(acopy) _projectiledata_hits))))
                        (:then
                          (Utilities|IsValid (Utilities|Array|Get(acopy) (Vehicle|GetAllPassengers _asfort_athena_vehicle))
                            (:Is%20Valid
                              (AthenaMovement|ServeronExitVehicle (Utilities|Array|Get(acopy) (Vehicle|GetAllPassengers _asfort_athena_vehicle)) (CallFunction|SetupVehicleExitData false)) ; Force 1 Player out of the Vehicle
                            )))
                        (:CastFailed
                          (bind _asfort_pawn (Utilities|Casting|CastToFortPawn (.HitActor (Collision|BreakHitResult (Utilities|Array|Get(acopy) _projectiledata_hits)))))
                          (CallFunction|HandlePlayerStates _asfort_pawn)))))))))
          (Ability|EndAbility))
        (:CastFailed
          _)))
    (:AvatarCastFailed
      _))
  (Ability|EndAbility))

(event Ability|EventOnEndAbility
  (params (bWasCancelled :type Boolean))
  (Ability|CommitAbility)
  (Class|FortAthenaAIBotController|EquipBestWeapon (Variables|InternalData|GetAIController) false)
  (Default|DisableProjectile)
  (bind _hookshotprojectile (Variables|Default|GetHookshotProjectile)
    (:then
      (Default|UnbindEventfromOnProjectileKilled _hookshotprojectile (EventDispatchers|CreateEvent :!event Custom|HandeProjectileKilled) :!class /ThornRipper/Gameplay/Items/HookshotGameplay/Gameplay/B_Prj_Hookshot.B_Prj_Hookshot_C)
      (Variables|Default|SetHookshotProjectile 0))
    (:else)))

(event Custom|HandeProjectileKilled
  ;[0.01535 0.09375 0.01502] HandleProjectileKilled
  (Default|DisableProjectile)
  (bind _hookshotprojectile (Variables|Default|GetHookshotProjectile)
    (:then
      (Default|UnbindEventfromOnProjectileKilled _hookshotprojectile (delegate Custom|HandeProjectileKilled) :!class /ThornRipper/Gameplay/Items/HookshotGameplay/Gameplay/B_Prj_Hookshot.B_Prj_Hookshot_C))
    (:else)))

(fn DisableProjectile
  ;[0.09106 0.02935 0.1042] DisableProjectile: Hides the projectile, disables the collision, and let's the rope know that it is time to end its behavior.
  (Ability|HasAuthority
    (:True
      (bind _hookshotprojectile (Variables|Default|GetHookshotProjectile)
        (:then
          (Class|BPrjHookshot|DisableProjectile :self _hookshotprojectile :!class /ThornRipper/Gameplay/Items/HookshotGameplay/Gameplay/B_Prj_Hookshot.B_Prj_Hookshot_C))
        (:else)))))

(fn GetCustomAbilitySourceTransform
  (return (+ (Transformation|GetActorLocation (Ability|GetAvatarActorfromActorInfo)) (* (Transformation|GetActorForwardVector (Ability|GetAvatarActorfromActorInfo)) "2500.000000,2500.000000,0.000000")) (Transformation|GetActorRotation (Ability|GetAvatarActorfromActorInfo)) "1.000000,1.000000,1.000000"))

(fn GetProjectileSpawnLocation
  (bind (_returnvalue_x _returnvalue_y) (Math|Vector|BreakVector (Transformation|GetActorForwardVector (CallFunction|GetAvatarasFortPlayerPawn))))
  (return (Utilities|Operators|Add (Transformation|GetActorLocation (CallFunction|GetAvatarasFortPlayerPawn)) (Math|Vector|MakeVector (* _returnvalue_x (ScalableFloat|GetValueatLevel (Variables|Default|GetProjectileSpawnDistanceXOffset))) (* _returnvalue_y (ScalableFloat|GetValueatLevel (Variables|Default|GetProjectileSpanDistanceYOffset))) 1e+02))))

(fn HandlePlayerStates
  (params (PlayerPawn :type Fort%20Pawn%20Object%20Reference))
  (bind _asfort_player_pawn (Utilities|Casting|CastToFortPlayerPawn PlayerPawn)
    (:then
      (bind _charactermovement (Variables|/Script/Engine.Character|GetCharacterMovement _asfort_player_pawn)
        (:then
          (switch Utilities|FlowControl|Switch|SwitchonEFortCustomMovement (Utilities|Enum|BytetoEnumEFortCustomMovement (Variables|/Script/Engine.CharacterMovementComponent|GetCustomMovementMode _charactermovement))
            (:Parachuting
              _)
            (:Skydiving
              _)
            (:SkydiveFollowing
              _)
            (:Swinging
              (Ability|SendGameplayEventtoActor PlayerPawn "(TagName=\"Event.Athena.StopSwinging\")" :!label _label))
            (:SwingingGrounded
              (goto _label)))
          (Parachute|EndSkydiving _asfort_player_pawn))
        (:else
          _)))
    (:CastFailed
      _))
  (return))

(fn CanActivateAbility
  (params (ActorInfo :type Gameplay%20Ability%20Actor%20Info%20Structure) (Handle :type Gameplay%20Ability%20Spec%20Handle%20Structure))
  (bind _returnvalue (CallFunction|GetAvatarasFortPlayerPawn))
  (bind _targetactor (Class|BPFLAICommonFunctions|GetTargetActorfromBlackboard _returnvalue))
  (Utilities|IsValid _targetactor
    (:Is%20Valid
      (return 0 (and (>= (Transformation|GetDistanceTo _returnvalue _targetactor) (ScalableFloat|GetValueatLevel (Variables|Default|GetMinDistanceToUseHook))) (<= (Transformation|GetDistanceTo _returnvalue _targetactor) (ScalableFloat|GetValueatLevel (Variables|Default|GetMaxDistanceToUseHook))))))
    (:Is%20Not%20Valid
      (return 0 false))))

(fn SetupVehicleExitData
  (params (DestroyVehicle :type Boolean))
  (return (Utilities|Struct|MakeVehicleExitData "ForceOnBlocking" DestroyVehicle (Utilities|Struct|MakeVehicleExitLocationSelectionData "ByClosestToGivenLocation" (Transformation|GetActorLocation (Variables|InternalData|GetAIPawn))))))
