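; Gameplay ability that plays the AxisThrow AI's (logged as BossD) custom intro montage, reacting to nearby pawns.
; Proximity events switch the montage between leaning and leaning back, with a speech bubble.
; Damage or an attack-proximity event triggers Stand and a failsafe delay, skipped on bypass, then ends the ability; the proximity event also turns the player hostile.
; Ending tags the pawn against replaying the intro and sends an IntroDone event.
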(event Ability|EventActivateAbility
  (bind (_montageinfo_montagetoplay _montageinfo_animplayrate _montageinfo_animrootmotiontranslationscale) (Utilities|Struct|BreakFortGameplayAbilityMontageInfo (Variables|GetMontageInfo_MontageToPlay)))
  (|CallParentFunction)
  (EventDispatchers|BindEventtoOnDamaged (Variables|InternalData|GetAIPawn) (EventDispatchers|CreateEvent :!event Custom|PawnWasDamaged))
  (Variables|Default|SetByPassIntro false)
  (Ability|Tasks|PlayMontageAndWaitNotify "None" _montageinfo_montagetoplay _montageinfo_animplayrate "None" true _montageinfo_animrootmotiontranslationscale
    (:then
      (Ability|Async|WaitGameplayEventToActor (Variables|InternalData|GetAIPawn) (Variables|Default|GetTag_EventPawnInProximity)
        (:then
          (bind (_asynctaskproxy _payload) (Ability|Tasks|WaitGameplayEvent (Variables|Default|GetEventTag_PlayerInAttackProximity) 0 true)
            (:EventReceived
              (Variables|Default|SetIsAttacking true)
              (bind _asfort_player_pawn (Utilities|Casting|CastToFortPlayerPawn (.Instigator (Utilities|Struct|BreakGameplayEventData _payload)))
                (:then
                  (Affiliation|ChangeAttitudeBetweenActors (Variables|InternalData|GetAIPawn) _asfort_player_pawn "Hostile"))
                (:CastFailed
                  _))
              (Variables|Default|SetByPassIntro (GameplayTags|HasTag (.InstigatorTags (Utilities|Struct|BreakGameplayEventData _payload)) "(TagName=\"Ability.AxisThrow.CustomIntro.ByPassAlert\")"))
              (CallFunction|MakeAIStandUp))))
        (:EventReceived
          (if (not (Variables|Default|GetIsAttacking))
            (Variables|Default|SetIsLeaning true)
            (Ability|Animation|MontageJumptoSection "StartLeaning")
            (Class|BPFLAICommonFunctions|PlaySpeechBubble (Variables|InternalData|GetAIPawn) (Variables|Default|GetTagQuery_SpeechBubbleProximity) true true)
            (Utilities|FlowControl|DoOnce false
              (:Completed
                (Ability|Async|WaitGameplayEventToActor (Variables|InternalData|GetAIPawn) (Variables|Default|GetTag_EventNoPawnInProximity)
                  (:EventReceived
                    (if (Variables|Default|GetByPassIntro)
                      _
                      (elif (Variables|Default|GetIsAttacking)
                        _
                        (else
                          (Ability|Animation|MontageJumptoSection "LeanBack")
                          (Variables|Default|SetIsLeaning false))))))))))))
    (:OnCompleted
      _)
    (:OnInterrupted
      (Fortnite|FortShippingLog self "BossD Custom Intro was Interupted or Cancelled! This should not have happened!!!!" true))
    (:OnCancelled
      _))
  (Ability|EndAbility))

(event Custom|MakeAIStandUp
  (Utilities|FlowControl|DoOnce false
    (:Completed
      (Class|BPFLAICommonFunctions|PlaySpeechBubble (Variables|InternalData|GetAIPawn) (Variables|Default|GetTagQuery_SpeechBubbleAttacked))
      (if (not (Variables|Default|GetByPassIntro))
        (Ability|Animation|MontageJumptoSection "Stand")
        (Utilities|FlowControl|Delay (ScalableFloat|GetValueatLevel (Variables|Default|GetIntroFailSafeTimerOnDamaged))))
      (Ability|EndAbility))))

(event Custom|PawnWasDamaged
  (params (DamagedActor :type Actor%20Object%20Reference) (Damage :type Float%20%28single-precision%29) (InstigatedBy :type Controller%20Object%20Reference) (DamageCauser :type Actor%20Object%20Reference) (HitLocation :type Vector) (FHitComponent :type Primitive%20Component%20Object%20Reference) (BoneName :type Name) (Momentum :type Vector))
  (Utilities|FlowControl|DoOnce false
    (:Completed
      (Class|BPFLAICommonFunctions|PlaySpeechBubble (Variables|InternalData|GetAIPawn) (Variables|Default|GetTagQuery_SpeechBubbleAttacked))
      (if (not (Variables|Default|GetByPassIntro))
        (Ability|Animation|MontageJumptoSection "Stand")
        (Utilities|FlowControl|Delay (ScalableFloat|GetValueatLevel (Variables|Default|GetIntroFailSafeTimerOnDamaged))))
      (Ability|EndAbility))))

(event Ability|EventOnEndAbility
  (params (bWasCancelled :type Boolean))
  (bind _aipawn (Variables|InternalData|GetAIPawn))
  ;[1 1 1] End Ability
  (Utilities|IsValid _aipawn
    (:Is%20Valid
      (Ability|GameplayEffect|AddLooseGameplayTags _aipawn (Variables|Default|GetTag_PreventIntro))
      (EventDispatchers|UnbindEventfromOnDamaged _aipawn (EventDispatchers|CreateEvent :!event Custom|PawnWasDamaged)))
    (:Is%20Not%20Valid
      _))
  (Ability|SendGameplayEvent (Variables|Default|GetEventTag_IntroDone))
  (|CallParentFunction bWasCancelled))

(fn InitVariables
  (|CallParentFunction)
  (Variables|Default|SetIsLeaning false)
  (Variables|Default|SetIsAttacking false))
