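; Gameplay ability for the AxisThrow NPC that knocks nearby targeted pawns and vehicles away from it.
; If enabled, it waits a tunable delay, runs target selection sourced at the pawn's Throne actor (or the pawn itself), then applies a gameplay effect container.
; Pawns are launched outward and upward; vehicles feed a similar vector into BroadcastAbilityImpactatLocation rather than a launch call.
; EndAbility is called only when push-back is disabled.
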(event Ability|EventActivateAbility
  (if (ScalableFloat|GetValueasBool (Variables|Default|GetPushBackEnabled))
    (Utilities|FlowControl|Delay (ScalableFloat|GetValueatLevel (Variables|Default|GetDelayBeforePushBack)))
    (bind (_asynctaskproxy _targetdata _applicationtag) (Ability|Tasks|DoTargetSelection "None" "(TagName=\"Targeting.ApplicationTag.0\")")
      (:Targeted
        (Ability|ApplyGameplayEffectContainer _targetdata _applicationtag)
        (for _array_element (Ability|TargetData|GetAllActorsfromTargetData _targetdata)
          (bind _asfort_pawn (Utilities|Casting|CastToFortPawn _array_element)
            (:then
              (Default|KnockbackPawns _asfort_pawn))
            (:CastFailed
              (bind _asfort_athena_vehicle (Utilities|Casting|CastToFortAthenaVehicle _array_element))
              (Default|KnockbackVehicles _asfort_athena_vehicle))))))
    (else
      (Ability|EndAbility))))

(event Ability|EventOnEndAbility
  (params (bWasCancelled :type Boolean)))

(fn GetCustomAbilitySourceTransform
  (bind _asbp_player_pawn_axis_throw (Utilities|Casting|CastToBP_PlayerPawn_AxisThrow (Ability|GetAvatarActorfromActorInfo)))
  (Utilities|IsValid (Variables|/FNM_AI_AxisThrow/NPCs/AxisThrow/Pawns/BP_PlayerPawn_AxisThrow.BP_PlayerPawn_AxisThrow_C|GetThroneActor _asbp_player_pawn_axis_throw)
    (:Is%20Valid
      (return (Transformation|GetActorTransform (Variables|/FNM_AI_AxisThrow/NPCs/AxisThrow/Pawns/BP_PlayerPawn_AxisThrow.BP_PlayerPawn_AxisThrow_C|GetThroneActor _asbp_player_pawn_axis_throw))))
    (:Is%20Not%20Valid
      (return (Transformation|GetActorTransform _asbp_player_pawn_axis_throw)))))

(fn KnockbackPawns
  (params (Pawn :type Fort%20Pawn%20Object%20Reference))
  (bind (_returnvalue_x _returnvalue_y) (Math|Vector|BreakVector (Math|Vector|Normalize (- (Transformation|GetActorLocation Pawn) (Transformation|GetActorLocation (Ability|GetAvatarActorfromActorInfo))))))
  (Character|LaunchCharacter Pawn (Math|Vector|MakeVector (* _returnvalue_x (ScalableFloat|GetValueatLevel (Variables|Default|GetKnockbackPawn_XYVelocity))) (* _returnvalue_y (ScalableFloat|GetValueatLevel (Variables|Default|GetKnockbackPawn_XYVelocity))) (ScalableFloat|GetValueatLevel (Variables|Default|GetKnockbackPawn_ZVelocity))) true true))

(fn KnockbackVehicles
  (params (FortAthenaVehicle :type Fort%20Athena%20Vehicle%20Object%20Reference))
  (bind (_returnvalue_x _returnvalue_y) (Math|Vector|BreakVector (Math|Vector|Normalize (- (Transformation|GetActorLocation FortAthenaVehicle) (Transformation|GetActorLocation (Ability|GetAvatarActorfromActorInfo))))))
  (Audio|BroadcastAbilityImpactatLocation 0 0 (Math|Vector|MakeVector (* _returnvalue_x (ScalableFloat|GetValueatLevel (Variables|Default|GetKnockbackVehicle_XYVelocity))) (* _returnvalue_y (ScalableFloat|GetValueatLevel (Variables|Default|GetKnockbackVehicle_XYVelocity))) (ScalableFloat|GetValueatLevel (Variables|Default|GetKnockbackVehicle_ZVelocity))) 777.0 "None" false))
