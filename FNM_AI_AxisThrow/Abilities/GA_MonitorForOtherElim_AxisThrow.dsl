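; Gameplay ability for the AxisThrow AI that monitors for eliminations it earns and plays a gloat speech bubble in response.
; On activation it caches the avatar as a valid Fort player pawn, then waits on that pawn for the Event.EarnedElimination gameplay event.
; When the event arrives, it calls PlaySpeechBubble from the BPFL AI common functions, using the speech-bubble tag query variable named for DBNO.
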(event Ability|EventActivateAbility
  (bind _outfortplayerpawn (CallFunction|GetAvatarasValidFortPlayerPawn)
    (:ValidFortPlayerPawn
      (Variables|Default|SetOut%20Fort%20Player%20Pawn _outfortplayerpawn)
      (Ability|Tasks|WaitGameplayEvent "(TagName=\"Event.EarnedElimination\")" _outfortplayerpawn
        (:EventReceived
          ; Elim gloat VO here
          (Class|BPFLAICommonFunctions|PlaySpeechBubble (Variables|Default|GetOut%20Fort%20Player%20Pawn) (Variables|Default|GetTagQuery_SpeechBubbleOnDBNO) true true))))))
