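; Gameplay ability for AxisThrow AI that watches its own avatar pawn for entering the DBNO state, apparently to react to self-elimination.
; On activation it validates the avatar as a Fort player pawn and binds to that pawn's DBNO state change event.
; When the pawn goes DBNO, it plays a speech bubble through the common AI function library, selected by a configurable gameplay tag query.
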(event Ability|EventActivateAbility
  (bind _outfortplayerpawn (CallFunction|GetAvatarasValidFortPlayerPawn)
    (:ValidFortPlayerPawn
      (EventDispatchers|BindEventtoOnDBNOStateChanged _outfortplayerpawn (delegate Custom|WentDBNO)))))

(event Custom|WentDBNO
  (params (FortPawn :type Fort%20Pawn%20Object%20Reference) (bInIsDBNO :type Boolean))
  (if bInIsDBNO
    (Class|BPFLAICommonFunctions|PlaySpeechBubble FortPawn (Variables|Default|GetTagQuery_SpeechBubbleOnDBNO) true true)))
